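# Animation Blueprint for the quadruped MinkCalico cosmetic companion, posing it by retargeting from another mesh.
# A short bool blend driven by bUseReverseRetarget selects the emote IK retargeter when true and the standard one otherwise.
# Each update compares the R_blink_pose curve to its cached value, raising the materials' Temporal Responsiveness parameter while the blink changes and lowering it otherwise.
# An unconnected Copy Pose From Mesh node remains in the AnimGraph.
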AnimBlueprint /CosmeticCompanions/Assets/Quadruped/MinkCalico/Meshes/MinkCalico_Quadruped_AnimBP
target_skeleton=&/MinkCalico_Quadruped_Skeleton
graphs [
 AnimGraph {nodes [
   Copy Pose From Mesh() {bCopyCurves=true, bUseAttachedParent=true}
   Output Pose(Result=Blend Poses by bool(
     BlendPose_0=Retarget Pose From Mesh!() -> (Pose:Pose Link Structure) {is=title AnimGraphNode_RetargetPoseFromMesh, iKRetargeterAsset=&/RTG_MinkCalico_Emote}
     BlendPose_1=Retarget Pose From Mesh() {iKRetargeterAsset=&/RTG_MinkCalico}
     BlendTime_0=0.2
     BlendTime_1=0.2
     bActiveValue=Get bUseReverseRetarget() -> (bUseReverseRetarget:bool)
    ))
  ]}
 EventGraph {nodes [
   Event BlueprintUpdateAnimation()
   n1 = GetCurveValue(CurveName=R_blink_pose)
   n2 = Get Blink Pose() -> (Blink Pose:float)
   > SetScalarParameterValueOnMaterials!(
    execute
    self:Mesh Component Object Reference=GetOwningComponent()
    ParameterName:Name=Temporal Responsiveness
    ParameterValue:float=SelectFloat(A=1, B=0.4, bPickA="Not Equal (Float)"(A=@n1, B=@n2))
   ) -> (then)
   > Set Blink Pose(Blink Pose:float=@n1) -> (Output_Get:float)
   TryGetPawnOwner()
  ]}
]
variables [Blink Pose {type=number}]
